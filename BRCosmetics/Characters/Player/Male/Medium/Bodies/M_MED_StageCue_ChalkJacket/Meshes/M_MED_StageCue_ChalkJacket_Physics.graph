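# Physics asset for the male medium StageCue ChalkJacket body cosmetic, setting up capsule bodies and joints for the jacket's dynamic bones.
# Eight chains around the front, back and sides each link a tiny master body to two larger segments.
# Tight twist and swing1 limits leave movement mostly on swing2, which widens at the lower segment.
# Pelvis, spine_01 and thigh capsules add body volumes, with only the thighs constrained, freely, to the pelvis.
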PhysicsAsset /BRCosmetics/Characters/Player/Male/Medium/Bodies/M_MED_StageCue_ChalkJacket/Meshes/M_MED_StageCue_ChalkJacket_Physics
bodies {
 dyn_jacket_01_L [{shapeType=Capsule, length=15.5, radius=5, shapeName=dyn_jacket_01_L_capsule, center=<3.0228 0 0>, rotation=<rot -90 172.875 97.1252>}]
 dyn_jacket_01_R [{shapeType=Capsule, length=15.5, radius=5, shapeName=dyn_jacket_01_R_capsule, center=<3.0228 0 0>, rotation=<rot -90 -165.9638 75.9636>}]
 dyn_jacket_02_L [{shapeType=Capsule, length=17, radius=5, shapeName=dyn_jacket_02_L_capsule, center=<4.1143 0 0>, rotation=<rot -90 180 90.0002>}]
 dyn_jacket_02_R [{shapeType=Capsule, length=17, radius=5, shapeName=dyn_jacket_02_R_capsule, center=<4.1143 0 0>, rotation=<rot -90 -165.9638 75.9636>}]
 dyn_jacket_MSTR_L [{shapeType=Capsule, length=1, radius=1, shapeName=dyn_jacket_MSTR_L_capsule, rotation=<rot -90 0 0>}]
 dyn_jacket_MSTR_R [{shapeType=Capsule, length=1, radius=1, shapeName=dyn_jacket_MSTR_R_capsule, rotation=<rot -90 0 0>}]
 dyn_jacket_bk_01 [{shapeType=Capsule, length=15.5, radius=5, shapeName=dyn_jacket_bk_01_capsule, center=<3.0228 0 0>, rotation=<rot -90 176.4237 93.5763>}]
 dyn_jacket_bk_02 [{shapeType=Capsule, length=14.5, radius=6, shapeName=dyn_jacket_bk_02_capsule, center=<4.1143 0 0>, rotation=<rot -90 177.1376 92.8624>}]
 dyn_jacket_bk_MSTR [{shapeType=Capsule, length=1, radius=1, shapeName=dyn_jacket_bk_MSTR_capsule, rotation=<rot -90 0 0>}]
 dyn_jacket_bk_mid_01_L [{
   shapeType=Capsule
   length=15.5
   radius=5
   shapeName=dyn_jacket_bk_mid_01_L_capsule
   center=<3.0228 0 0>
   rotation=<rot -90 -165.9638 75.9639>
  }]
 dyn_jacket_bk_mid_01_R [{
   shapeType=Capsule
   length=15.5
   radius=5
   shapeName=dyn_jacket_bk_mid_01_R_capsule
   center=<3.0228 0 0>
   rotation=<rot -90 -165.9638 75.9636>
  }]
 dyn_jacket_bk_mid_02_L [{shapeType=Capsule, length=14.5, radius=6, shapeName=dyn_jacket_bk_mid_02_L_capsule, center=<4.1143 0 0>, rotation=<rot -90 180 90.0001>}]
 dyn_jacket_bk_mid_02_R [{shapeType=Capsule, length=14.5, radius=6, shapeName=dyn_jacket_bk_mid_02_R_capsule, center=<4.1143 0 0>, rotation=<rot -90 -172.875 82.8749>}]
 dyn_jacket_bk_mid_MSTR_L [{shapeType=Capsule, length=1, radius=1, shapeName=dyn_jacket_bk_mid_MSTR_L_capsule, rotation=<rot -90 0 0>}]
 dyn_jacket_bk_mid_MSTR_R [{shapeType=Capsule, length=1, radius=1, shapeName=dyn_jacket_bk_mid_MSTR_R_capsule, rotation=<rot -90 0 0>}]
 dyn_jacket_fr_01 [{shapeType=Capsule, length=15.5, radius=5, shapeName=dyn_jacket_fr_01_capsule, center=<3.0228 0 0>, rotation=<rot -90 172.875 97.125>}]
 dyn_jacket_fr_02 [{shapeType=Capsule, length=11.5, radius=5, shapeName=dyn_jacket_fr_02_capsule, center=<4.1143 0 0>, rotation=<rot -90 172.875 97.125>}]
 dyn_jacket_fr_MSTR [{shapeType=Capsule, length=1, radius=1, shapeName=dyn_jacket_fr_MSTR_capsule, rotation=<rot -90 0 0>}]
 dyn_jacket_fr_mid_01_L [{shapeType=Capsule, length=15.5, radius=5, shapeName=dyn_jacket_fr_mid_01_L_capsule, center=<3.0228 0 0>, rotation=<rot -90 172.875 97.1251>}]
 dyn_jacket_fr_mid_01_R [{
   shapeType=Capsule
   length=15.5
   radius=5
   shapeName=dyn_jacket_fr_mid_01_R_capsule
   center=<3.0228 0 0>
   rotation=<rot -90 -165.9638 75.9636>
  }]
 dyn_jacket_fr_mid_02_L [{shapeType=Capsule, length=15.5, radius=5, shapeName=dyn_jacket_fr_mid_02_L_capsule, center=<4.1143 0 0>, rotation=<rot -90 172.875 97.1251>}]
 dyn_jacket_fr_mid_02_R [{
   shapeType=Capsule
   length=15.5
   radius=5
   shapeName=dyn_jacket_fr_mid_02_R_capsule
   center=<4.1143 0 0>
   rotation=<rot -90 -165.9638 75.9636>
  }]
 dyn_jacket_fr_mid_MSTR_L [{shapeType=Capsule, length=1, radius=1, shapeName=dyn_jacket_fr_mid_MSTR_L_capsule, rotation=<rot -90 0 0>}]
 dyn_jacket_fr_mid_MSTR_R [{shapeType=Capsule, length=1, radius=1, shapeName=dyn_jacket_fr_mid_MSTR_R_capsule, rotation=<rot -90 0 0>}]
 pelvis [{shapeType=Capsule, length=5, radius=9.5, shapeName=pelvis_capsule, center=<0 -1.0988 0>}]
 spine_01 [{shapeType=Capsule, length=5, radius=8, shapeName=spine_01_capsule, center=<4.8031 -1.0988 0>}]
 thigh_l [{shapeType=Capsule, length=21.5, radius=8.5, shapeName=thigh_l_capsule, center=<-20 0 0>, rotation=<rot 90 89.6476 89.6479>}]
 thigh_r [{shapeType=Capsule, length=20, radius=8.5, shapeName=thigh_r_capsule, center=<20 0 0>, rotation=<rot 90 89.6476 89.6478>}]
}
constraints [
 {
  bone1Name=thigh_l
  bone2Name=pelvis
  swing1LimitDegrees=45
  swing1Motion=Free
  swing2LimitDegrees=45
  swing2Motion=Free
  twistLimitDegrees=45
  twistMotion=Free
 }
 {
  bone1Name=thigh_r
  bone2Name=pelvis
  swing1LimitDegrees=45
  swing1Motion=Free
  swing2LimitDegrees=45
  swing2Motion=Free
  twistLimitDegrees=45
  twistMotion=Free
 }
 {bone1Name=dyn_jacket_01_L, bone2Name=dyn_jacket_MSTR_L, swing1LimitDegrees=2.5, swing2LimitDegrees=15, twistLimitDegrees=2.5}
 {bone1Name=dyn_jacket_02_L, bone2Name=dyn_jacket_01_L, swing1LimitDegrees=2.5, swing2LimitDegrees=25, twistLimitDegrees=2.5}
 {bone1Name=dyn_jacket_01_R, bone2Name=dyn_jacket_MSTR_R, swing1LimitDegrees=2.5, swing2LimitDegrees=15, twistLimitDegrees=2.5}
 {bone1Name=dyn_jacket_02_R, bone2Name=dyn_jacket_01_R, swing1LimitDegrees=2.5, swing2LimitDegrees=25, twistLimitDegrees=2.5}
 {bone1Name=dyn_jacket_bk_01, bone2Name=dyn_jacket_bk_MSTR, swing1LimitDegrees=2.5, swing2LimitDegrees=15, twistLimitDegrees=2.5}
 {bone1Name=dyn_jacket_bk_02, bone2Name=dyn_jacket_bk_01, swing1LimitDegrees=2.5, swing2LimitDegrees=25, twistLimitDegrees=2.5}
 {bone1Name=dyn_jacket_bk_mid_01_L, bone2Name=dyn_jacket_bk_mid_MSTR_L, swing1LimitDegrees=2.5, swing2LimitDegrees=15, twistLimitDegrees=2.5}
 {bone1Name=dyn_jacket_bk_mid_02_L, bone2Name=dyn_jacket_bk_mid_01_L, swing1LimitDegrees=2.5, swing2LimitDegrees=25, twistLimitDegrees=2.5}
 {bone1Name=dyn_jacket_bk_mid_01_R, bone2Name=dyn_jacket_bk_mid_MSTR_R, swing1LimitDegrees=2.5, swing2LimitDegrees=15, twistLimitDegrees=2.5}
 {bone1Name=dyn_jacket_bk_mid_02_R, bone2Name=dyn_jacket_bk_mid_01_R, swing1LimitDegrees=2.5, swing2LimitDegrees=25, twistLimitDegrees=2.5}
 {bone1Name=dyn_jacket_fr_01, bone2Name=dyn_jacket_fr_MSTR, swing1LimitDegrees=2.5, swing2LimitDegrees=15, twistLimitDegrees=2.5}
 {bone1Name=dyn_jacket_fr_02, bone2Name=dyn_jacket_fr_01, swing1LimitDegrees=2.5, swing2LimitDegrees=25, twistLimitDegrees=2.5}
 {bone1Name=dyn_jacket_fr_mid_01_L, bone2Name=dyn_jacket_fr_mid_MSTR_L, swing1LimitDegrees=2.5, swing2LimitDegrees=15, twistLimitDegrees=2.5}
 {bone1Name=dyn_jacket_fr_mid_02_L, bone2Name=dyn_jacket_fr_mid_01_L, swing1LimitDegrees=2.5, swing2LimitDegrees=25, twistLimitDegrees=2.5}
 {bone1Name=dyn_jacket_fr_mid_01_R, bone2Name=dyn_jacket_fr_mid_MSTR_R, swing1LimitDegrees=2.5, swing2LimitDegrees=15, twistLimitDegrees=2.5}
 {bone1Name=dyn_jacket_fr_mid_02_R, bone2Name=dyn_jacket_fr_mid_01_R, swing1LimitDegrees=2.5, swing2LimitDegrees=25, twistLimitDegrees=2.5}
]
