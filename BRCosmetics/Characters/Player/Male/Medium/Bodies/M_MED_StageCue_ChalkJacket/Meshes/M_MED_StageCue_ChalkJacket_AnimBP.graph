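# Post-body AnimBP for the M_MED_StageCue_ChalkJacket cosmetic, adding secondary motion and pose fixes to the jacket.
# It copies the parent mesh pose, constrains and simulates buckles and zipper with AnimDynamics, then offsets sleeve, collar and jacket bones.
# Cached main, crouch and grind-rail variants blend by state into a pelvis-space RigidBody, then a Control Rig.
# The event graph computes gravity/RBAN settings, jump/skydive values and hand, head, thigh rotation factors.
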AnimBlueprint /BRCosmetics/Characters/Player/Male/Medium/Bodies/M_MED_StageCue_ChalkJacket/Meshes/M_MED_StageCue_ChalkJacket_AnimBP
target_skeleton=&/M_MED_StageCue_ChalkJacket_Skeleton
graphs [
 AnimGraph {nodes [
   n1 = Blend Poses by bool(BlendPose_0=Use cached pose 'crouching'(), BlendPose_1=Use cached pose 'main pose'(), bActiveValue=Get bIsCrouching()) {blendType=Linear}
   n2 = Use cached pose 'grind rail'()
   n3 = Get bIsOnGrindRail()
   n4 = RigidBody!(
    ComponentPose=Local To Component(LocalPose=Blend Poses by bool(BlendPose_0=@n2, BlendPose_1=@n1, bActiveValue=@n3) {blendType=Linear})
    Alpha
    OverrideWorldGravity:Vector="0,0,0"
    ComponentLinearAccScale:Vector="0.35,0.35,0.35"
    ComponentLinearVelScale:Vector="0.4,0.4,0.4"
   ) -> (Pose) {
    bOverrideWorldGravity=true
    overridePhysicsAsset=&/M_MED_StageCue_ChalkJacket_Physics
    simulationSpace=BaseBoneSpace
    alphaBoolBlend {blendOption=Linear}
    baseBoneRef {boneName=pelvis}
    componentAppliedLinearAccClamp=<1000 1000 1000>
   }
   n5 = Control Rig(Source=Component To Local(ComponentPose=@n4)) {
    alphaBoolBlend {blendOption=Linear}
    controlRigAssetReference {blueprintRigClass=&/M_MED_StageCue_ChalkJacket_ControlRig.M_MED_StageCue_ChalkJacket_ControlRig_C}
   }
   Output Pose(Result=@n5) {layerGroup=None}
   n6 = "Constraint - dyn_buckle_upper_CST_L (PPPPP)"!(
    ComponentPose=Local To Component(LocalPose=Copy Pose From Mesh() {bUseAttachedParent=true})
    Alpha=1
    ConstraintWeights_0=0.0115
    ConstraintWeights_1:float=0.0035
    ConstraintWeights_2:float=0.0541
    ConstraintWeights_3:float=0.1671
    ConstraintWeights_4:float=0.7638
   ) -> (Pose) {
    alphaBoolBlend {blendOption=Linear}
    boneToModify {boneName=dyn_buckle_upper_CST_L}
    constraintSetup [
     {targetBone {boneName=spine_04}}
     {targetBone {boneName=spine_05}}
     {targetBone {boneName=clavicle_l}}
     {targetBone {boneName=upperarm_twist_01_l}}
     {targetBone {boneName=upperarm_twist_02_l}}
    ]
    constraintWeights [1,1,1,1,1]
   }
   n7 = "Constraint - dyn_buckle_lower_CST_L (PPPPPPPP)"!(
    ComponentPose=@n6
    Alpha=1
    ConstraintWeights_0=0.0054
    ConstraintWeights_1:float=0.0012
    ConstraintWeights_2:float=0.0242
    ConstraintWeights_3:float=0.0024
    ConstraintWeights_4:float=0.0779
    ConstraintWeights_5:float=0.8889
    ConstraintWeights_6:float=0.0001
    ConstraintWeights_7:float=0
   ) -> (Pose) {
    alphaBoolBlend {blendOption=Linear}
    boneToModify {boneName=dyn_buckle_lower_CST_L}
    constraintSetup [
     {targetBone {boneName=spine_04}}
     {targetBone {boneName=spine_05}}
     {targetBone {boneName=clavicle_l}}
     {targetBone {boneName=upperarm_l}}
     {targetBone {boneName=upperarm_twist_01_l}}
     {targetBone {boneName=upperarm_twist_02_l}}
     {targetBone {boneName=lowerarm_l}}
     {targetBone {boneName=sleeve_L}}
    ]
    constraintWeights [1,1,1,1,1,1,1,1]
   }
   n8 = "Constraint - dyn_buckle_upper_CST_R (PPPPP)"!(
    ComponentPose=@n7
    Alpha=1
    ConstraintWeights_0=0.0115
    ConstraintWeights_1:float=0.0035
    ConstraintWeights_2:float=0.0541
    ConstraintWeights_3:float=0.1671
    ConstraintWeights_4:float=0.7638
   ) -> (Pose) {
    alphaBoolBlend {blendOption=Linear}
    boneToModify {boneName=dyn_buckle_upper_CST_R}
    constraintSetup [
     {targetBone {boneName=spine_04}}
     {targetBone {boneName=spine_05}}
     {targetBone {boneName=clavicle_r}}
     {targetBone {boneName=upperarm_twist_01_r}}
     {targetBone {boneName=upperarm_twist_02_r}}
    ]
    constraintWeights [1,1,1,1,1]
   }
   n9 = "Constraint - dyn_buckle_lower_CST_R (PPPPPPPP)"!(
    ComponentPose=@n8
    Alpha=1
    ConstraintWeights_0=0.0061
    ConstraintWeights_1:float=0.0013
    ConstraintWeights_2:float=0.0257
    ConstraintWeights_3:float=0.0025
    ConstraintWeights_4:float=0.0829
    ConstraintWeights_5:float=0.8814
    ConstraintWeights_6:float=0.0001
    ConstraintWeights_7:float=0
   ) -> (Pose) {
    alphaBoolBlend {blendOption=Linear}
    boneToModify {boneName=dyn_buckle_lower_CST_R}
    constraintSetup [
     {targetBone {boneName=spine_04}}
     {targetBone {boneName=spine_05}}
     {targetBone {boneName=clavicle_r}}
     {targetBone {boneName=upperarm_r}}
     {targetBone {boneName=upperarm_twist_01_r}}
     {targetBone {boneName=upperarm_twist_02_r}}
     {targetBone {boneName=lowerarm_r}}
     {targetBone {boneName=sleeve_R}}
    ]
    constraintWeights [1,1,1,1,1,1,1,1]
   }
   n10 = "Constraint - dyn_zipper_CST (PPPPP)"!(
    ComponentPose=@n9
    Alpha=1
    ConstraintWeights_0=0.0001
    ConstraintWeights_1:float=0.008
    ConstraintWeights_2:float=0.991
    ConstraintWeights_3:float=0.0003
    ConstraintWeights_4:float=0.0007
   ) -> (Pose) {
    alphaBoolBlend {blendOption=Linear}
    boneToModify {boneName=dyn_zipper_CST}
    constraintSetup [
     {targetBone {boneName=spine_04}}
     {targetBone {boneName=spine_05}}
     {targetBone {boneName=collar_fr_01}}
     {targetBone {boneName=collar_01_R}}
     {targetBone {boneName=collar_01_L}}
    ]
    constraintWeights [1,1,1,1,1]
   }
   n11 = Anim Dynamics - Bone: dyn_buckle_upper_01_R!(ComponentPose=@n10, Alpha=1, GravityOverride:Vector="0,0,-980") -> (Pose) {
    angularBiasOverride=1
    angularDampingOverride=0.99
    angularSpringConstant=500
    bOverrideAngularBias=true
    bOverrideAngularDamping=true
    bUseGravityOverride=true
    simulationSpace=RootRelative
    alphaBoolBlend {blendOption=Linear}
    boundBone {boneName=dyn_buckle_upper_01_R}
    componentAppliedLinearAccClamp=<1000 1000 1000>
    componentLinearAccScale=<0.7 0.7 0.7>
    componentLinearVelScale=<0.7 0.7 0.7>
    physicsBodyDefinitions [{
      boundBone {boneName=dyn_buckle_upper_01_R}
      boxExtents=<10 10 10>
      constraintSetup {angularLimitsMax=<0 20 0>, angularLimitsMin=<0 -20 0>, angularTarget=<1 0 0>}
      localJointOffset=<1 0 0>
     }]
   }
   n12 = Anim Dynamics - Bone: dyn_buckle_lower_01_R!(ComponentPose=@n11, Alpha=1, GravityOverride:Vector="0,0,-980") -> (Pose) {
    angularBiasOverride=1
    angularDampingOverride=0.99
    angularSpringConstant=500
    bOverrideAngularBias=true
    bOverrideAngularDamping=true
    bUseGravityOverride=true
    simulationSpace=RootRelative
    alphaBoolBlend {blendOption=Linear}
    boundBone {boneName=dyn_buckle_lower_01_R}
    componentAppliedLinearAccClamp=<1000 1000 1000>
    componentLinearAccScale=<0.7 0.7 0.7>
    componentLinearVelScale=<0.7 0.7 0.7>
    physicsBodyDefinitions [{
      boundBone {boneName=dyn_buckle_lower_01_R}
      boxExtents=<10 10 10>
      constraintSetup {angularLimitsMax=<0 20 0>, angularLimitsMin=<0 -20 0>, angularTarget=<1 0 0>}
      localJointOffset=<1 0 0>
     }]
   }
   n13 = Anim Dynamics - Bone: dyn_buckle_upper_01_L!(ComponentPose=@n12, Alpha=1, GravityOverride:Vector="0,0,-980") -> (Pose) {
    angularBiasOverride=1
    angularDampingOverride=0.99
    angularSpringConstant=500
    bOverrideAngularBias=true
    bOverrideAngularDamping=true
    bUseGravityOverride=true
    simulationSpace=RootRelative
    alphaBoolBlend {blendOption=Linear}
    boundBone {boneName=dyn_buckle_upper_01_L}
    componentAppliedLinearAccClamp=<1000 1000 1000>
    componentLinearAccScale=<0.7 0.7 0.7>
    componentLinearVelScale=<0.7 0.7 0.7>
    physicsBodyDefinitions [{
      boundBone {boneName=dyn_buckle_upper_01_L}
      boxExtents=<10 10 10>
      constraintSetup {angularLimitsMax=<0 20 0>, angularLimitsMin=<0 -20 0>, angularTarget=<1 0 0>}
      localJointOffset=<1 0 0>
     }]
   }
   n14 = Anim Dynamics - Bone: dyn_buckle_lower_01_L!(ComponentPose=@n13, Alpha=1, GravityOverride:Vector="0,0,-980") -> (Pose) {
    angularBiasOverride=0.99
    angularDampingOverride=0.99
    angularSpringConstant=500
    bOverrideAngularBias=true
    bOverrideAngularDamping=true
    bUseGravityOverride=true
    simulationSpace=RootRelative
    alphaBoolBlend {blendOption=Linear}
    boundBone {boneName=dyn_buckle_lower_01_L}
    componentAppliedLinearAccClamp=<1000 1000 1000>
    componentLinearAccScale=<0.7 0.7 0.7>
    componentLinearVelScale=<0.7 0.7 0.7>
    physicsBodyDefinitions [{
      boundBone {boneName=dyn_buckle_lower_01_L}
      boxExtents=<10 10 10>
      constraintSetup {angularLimitsMax=<0 20 0>, angularLimitsMin=<0 -20 0>, angularTarget=<1 0 0>}
      localJointOffset=<1 0 0>
     }]
   }
   n15 = Anim Dynamics - Bone: dyn_zipper_01!(ComponentPose=@n14, Alpha=1, GravityOverride:Vector="0,0,-980") -> (Pose) {
    angularBiasOverride=0.99
    angularDampingOverride=0.99
    angularSpringConstant=500
    bOverrideAngularBias=true
    bOverrideAngularDamping=true
    bUseGravityOverride=true
    simulationSpace=RootRelative
    alphaBoolBlend {blendOption=Linear}
    boundBone {boneName=dyn_zipper_01}
    componentAppliedLinearAccClamp=<1000 1000 1000>
    componentLinearAccScale=<0.7 0.7 0.7>
    componentLinearVelScale=<0.7 0.7 0.7>
    physicsBodyDefinitions [{
      boundBone {boneName=dyn_zipper_01}
      boxExtents=<10 10 10>
      constraintSetup {angularLimitsMax=<0 15 15>, angularLimitsMin=<0 -15 -15>, angularTarget=<1 0 0>}
      localJointOffset=<1 0 0>
     }]
   }
   n16 = "Transform (Modify) Bone - Bone: sleeve_L"!(ComponentPose=@n15, Alpha=1, Translation="-2.5,0,0") -> (Pose) {
    rotationSpace=BCS_BoneSpace
    translationMode=BMM_Additive
    translationSpace=BCS_BoneSpace
    alphaBoolBlend {blendOption=Linear}
    boneToModify {boneName=sleeve_L}
    translation {x=-2.5}
   }
   n17 = "Transform (Modify) Bone - Bone: sleeve_L"(ComponentPose=@n16, Alpha=1, Translation="-4,0,0") {
    rotationSpace=BCS_BoneSpace
    translationMode=BMM_Additive
    translationSpace=BCS_BoneSpace
    alphaBoolBlend {blendOption=Linear}
    boneToModify {boneName=sleeve_L}
    translation {x=-4}
   }
   n18 = "Transform (Modify) Bone - Bone: sleeve_R"!(ComponentPose=@n17, Alpha=1, Translation="2.5,0,0") -> (Pose) {
    rotationSpace=BCS_BoneSpace
    translationMode=BMM_Additive
    translationSpace=BCS_BoneSpace
    alphaBoolBlend {blendOption=Linear}
    boneToModify {boneName=sleeve_R}
    translation {x=2.5}
   }
   n19 = "Transform (Modify) Bone - Bone: sleeve_R"(ComponentPose=@n18, Alpha=1, Translation="1,0,0") {
    rotationSpace=BCS_BoneSpace
    translationMode=BMM_Additive
    translationSpace=BCS_BoneSpace
    alphaBoolBlend {blendOption=Linear}
    boneToModify {boneName=sleeve_R}
    translation {x=1}
   }
   n20 = "Transform (Modify) Bone - Bone: collar_bk_01"!(ComponentPose=@n19, Alpha=1, Rotation="70,0,0") -> (Pose) {
    rotationMode=BMM_Additive
    rotationSpace=BCS_BoneSpace
    alphaBoolBlend {blendOption=Linear}
    boneToModify {boneName=collar_bk_01}
    rotation {pitch=70}
   }
   n21 = "Transform (Modify) Bone - Bone: collar_bk_mid_01_R"!(ComponentPose=@n20, Alpha=1, Rotation="50,0,50") -> (Pose) {
    rotationMode=BMM_Additive
    rotationSpace=BCS_BoneSpace
    alphaBoolBlend {blendOption=Linear}
    boneToModify {boneName=collar_bk_mid_01_R}
    rotation {pitch=50, roll=50}
   }
   n22 = "Transform (Modify) Bone - Bone: collar_bk_01"(ComponentPose=@n21, Alpha=1, Rotation="70,0,0") {
    rotationMode=BMM_Additive
    rotationSpace=BCS_BoneSpace
    alphaBoolBlend {blendOption=Linear}
    boneToModify {boneName=collar_bk_01}
    rotation {pitch=70}
   }
   n23 = "Transform (Modify) Bone - Bone: collar_bk_mid_01_L"!(ComponentPose=@n22, Alpha=1, Rotation="30,0,0") -> (Pose) {
    rotationMode=BMM_Additive
    rotationSpace=BCS_BoneSpace
    alphaBoolBlend {blendOption=Linear}
    boneToModify {boneName=collar_bk_mid_01_L}
    rotation {pitch=30}
   }
   n24 = "Transform (Modify) Bone - Bone: collar_bk_mid_01_R"(ComponentPose=@n23, Alpha=1, Rotation="30,0,0") {
    rotationMode=BMM_Additive
    rotationSpace=BCS_BoneSpace
    alphaBoolBlend {blendOption=Linear}
    boneToModify {boneName=collar_bk_mid_01_R}
    rotation {pitch=30}
   }
   n25 = "Transform (Modify) Bone - Bone: collar_01_L"!(ComponentPose=@n24, Alpha=1, Rotation="80,0,0") -> (Pose) {
    rotationMode=BMM_Additive
    rotationSpace=BCS_BoneSpace
    alphaBoolBlend {blendOption=Linear}
    boneToModify {boneName=collar_01_L}
    rotation {pitch=80}
   }
   n26 = "Transform (Modify) Bone - Bone: collar_01_R"!(ComponentPose=@n25, Alpha=1, Rotation="80,0,0") -> (Pose) {
    rotationMode=BMM_Additive
    rotationSpace=BCS_BoneSpace
    alphaBoolBlend {blendOption=Linear}
    boneToModify {boneName=collar_01_R}
    rotation {pitch=80}
   }
   n27 = "Transform (Modify) Bone - Bone: collar_bk_mid_01_L"(ComponentPose=@n26, Alpha=1, Rotation="40,0,0") {
    rotationMode=BMM_Additive
    rotationSpace=BCS_BoneSpace
    alphaBoolBlend {blendOption=Linear}
    boneToModify {boneName=collar_bk_mid_01_L}
    rotation {pitch=40}
   }
   n28 = "Transform (Modify) Bone - Bone: collar_bk_mid_01_R"(ComponentPose=@n27, Alpha=1, Rotation="40,0,0") {
    rotationMode=BMM_Additive
    rotationSpace=BCS_BoneSpace
    alphaBoolBlend {blendOption=Linear}
    boneToModify {boneName=collar_bk_mid_01_R}
    rotation {pitch=40}
   }
   n29 = "Transform (Modify) Bone - Bone: collar_bk_01"(ComponentPose=@n28, Alpha=1, Rotation="40,0,0") {
    rotationMode=BMM_Additive
    rotationSpace=BCS_BoneSpace
    alphaBoolBlend {blendOption=Linear}
    boneToModify {boneName=collar_bk_01}
    rotation {pitch=40}
   }
   n30 = "Transform (Modify) Bone - Bone: collar_bk_01"(ComponentPose=@n29, Alpha=1, Rotation="40,0,0") {
    rotationMode=BMM_Additive
    rotationSpace=BCS_BoneSpace
    alphaBoolBlend {blendOption=Linear}
    boneToModify {boneName=collar_bk_01}
    rotation {pitch=40}
   }
   n31 = "Transform (Modify) Bone - Bone: collar_bk_mid_01_R"(ComponentPose=@n30, Alpha=1, Rotation="40,0,0") {
    rotationMode=BMM_Additive
    rotationSpace=BCS_BoneSpace
    alphaBoolBlend {blendOption=Linear}
    boneToModify {boneName=collar_bk_mid_01_R}
    rotation {pitch=40}
   }
   n32 = "Transform (Modify) Bone - Bone: collar_bk_mid_01_L"(ComponentPose=@n31, Alpha=1, Rotation="40,0,0") {
    rotationMode=BMM_Additive
    rotationSpace=BCS_BoneSpace
    alphaBoolBlend {blendOption=Linear}
    boneToModify {boneName=collar_bk_mid_01_L}
    rotation {pitch=40}
   }
   n33 = "Transform (Modify) Bone - Bone: collar_bk_01"!(ComponentPose=@n32, Alpha=1, Translation:Vector="0,0,5.5", Rotation="-25,0,0") -> (Pose) {
    rotationMode=BMM_Additive
    rotationSpace=BCS_BoneSpace
    translationMode=BMM_Additive
    translationSpace=BCS_BoneSpace
    alphaBoolBlend {blendOption=Linear}
    alphaScaleBiasClamp {interpSpeedDecreasing=0.2, interpSpeedIncreasing=20}
    boneToModify {boneName=collar_bk_01}
    rotation {pitch=-25}
    translation {z=5.5}
   }
   n34 = "Transform (Modify) Bone - Bone: collar_bk_mid_01_R"!(ComponentPose=@n33, Alpha=1, Translation:Vector="0,0,5", Rotation="-25,0,0") -> (Pose) {
    rotationMode=BMM_Additive
    rotationSpace=BCS_BoneSpace
    translationMode=BMM_Additive
    translationSpace=BCS_BoneSpace
    alphaBoolBlend {blendOption=Linear}
    alphaScaleBiasClamp {interpSpeedDecreasing=0.2, interpSpeedIncreasing=20}
    boneToModify {boneName=collar_bk_mid_01_R}
    rotation {pitch=-25}
    translation {z=5}
   }
   n35 = "Transform (Modify) Bone - Bone: collar_bk_mid_01_L"!(ComponentPose=@n34, Alpha=1, Translation:Vector="0,0,5", Rotation="-25,0,0") -> (Pose) {
    rotationMode=BMM_Additive
    rotationSpace=BCS_BoneSpace
    translationMode=BMM_Additive
    translationSpace=BCS_BoneSpace
    alphaBoolBlend {blendOption=Linear}
    alphaScaleBiasClamp {interpSpeedDecreasing=0.2, interpSpeedIncreasing=20}
    boneToModify {boneName=collar_bk_mid_01_L}
    rotation {pitch=-25}
    translation {z=5}
   }
   n36 = "Constraint - dyn_jacket_bk_mid_MSTR_R (PP)"!(ComponentPose=@n35, Alpha=1, ConstraintWeights_0=0.5, ConstraintWeights_1:float=0.5) -> (Pose) {
    alphaBoolBlend {blendOption=Linear}
    alphaScaleBiasClamp {bInterpResult=true}
    boneToModify {boneName=dyn_jacket_bk_mid_MSTR_R}
    constraintSetup [{targetBone {boneName=pelvis}}, {targetBone {boneName=spine_01}}]
    constraintWeights [1,1]
   }
   n37 = "Constraint - dyn_jacket_bk_MSTR (PP)"!(ComponentPose=@n36, Alpha=1, ConstraintWeights_0=0.5, ConstraintWeights_1:float=0.5) -> (Pose) {
    alphaBoolBlend {blendOption=Linear}
    alphaScaleBiasClamp {bInterpResult=true}
    boneToModify {boneName=dyn_jacket_bk_MSTR}
    constraintSetup [{targetBone {boneName=pelvis}}, {targetBone {boneName=spine_01}}]
    constraintWeights [1,1]
   }
   n38 = "Constraint - dyn_jacket_bk_mid_MSTR_L (PP)"!(ComponentPose=@n37, Alpha=1, ConstraintWeights_0=0.5, ConstraintWeights_1:float=0.5) -> (Pose) {
    alphaBoolBlend {blendOption=Linear}
    alphaScaleBiasClamp {bInterpResult=true}
    boneToModify {boneName=dyn_jacket_bk_mid_MSTR_L}
    constraintSetup [{targetBone {boneName=pelvis}}, {targetBone {boneName=spine_01}}]
    constraintWeights [1,1]
   }
   n39 = "Constraint - dyn_jacket_fr_MSTR (P)"(ComponentPose=@n38, Alpha=1, ConstraintWeights_0=1) {
    alphaBoolBlend {blendOption=Linear}
    alphaScaleBiasClamp {bInterpResult=true}
    boneToModify {boneName=dyn_jacket_fr_MSTR}
    constraintSetup [{targetBone {boneName=pelvis}}]
    constraintWeights [1]
   }
   n40 = "Constraint - dyn_jacket_fr_mid_MSTR_L (P)"(ComponentPose=@n39, Alpha=1, ConstraintWeights_0=1) {
    alphaBoolBlend {blendOption=Linear}
    alphaScaleBiasClamp {bInterpResult=true}
    boneToModify {boneName=dyn_jacket_fr_mid_MSTR_L}
    constraintSetup [{targetBone {boneName=pelvis}}]
    constraintWeights [1]
   }
   n41 = "Constraint - dyn_jacket_fr_mid_MSTR_R (P)"(ComponentPose=@n40, Alpha=1, ConstraintWeights_0=1) {
    alphaBoolBlend {blendOption=Linear}
    alphaScaleBiasClamp {bInterpResult=true}
    boneToModify {boneName=dyn_jacket_fr_mid_MSTR_R}
    constraintSetup [{targetBone {boneName=pelvis}}]
    constraintWeights [1]
   }
   n42 = "Transform (Modify) Bone - Bone: dyn_jacket_fr_MSTR"!(ComponentPose=@n41, Alpha=1, Rotation="20,0,0.001") -> (Pose) {
    rotationMode=BMM_Additive
    rotationSpace=BCS_BoneSpace
    alphaBoolBlend {blendOption=Linear}
    boneToModify {boneName=dyn_jacket_fr_MSTR}
    rotation {pitch=20, roll=0.001}
   }
   n43 = "Transform (Modify) Bone - Bone: dyn_jacket_fr_mid_MSTR_L"!(ComponentPose=@n42, Alpha=1, Rotation="22.846,35.242,16.808") -> (Pose) {
    rotationMode=BMM_Additive
    rotationSpace=BCS_BoneSpace
    alphaBoolBlend {blendOption=Linear}
    boneToModify {boneName=dyn_jacket_fr_mid_MSTR_L}
    rotation=<rot 22.8461 35.2421 16.808>
   }
   n44 = "Transform (Modify) Bone - Bone: dyn_jacket_MSTR_L"!(ComponentPose=@n43, Alpha=1, Rotation="-16.149,16.086,-23.159") -> (Pose) {
    rotationMode=BMM_Additive
    rotationSpace=BCS_BoneSpace
    alphaBoolBlend {blendOption=Linear}
    boneToModify {boneName=dyn_jacket_MSTR_L}
    rotation=<rot -16.1489 16.0862 -23.159>
   }
   n45 = "Transform (Modify) Bone - Bone: dyn_jacket_fr_mid_MSTR_R"!(ComponentPose=@n44, Alpha=1, Rotation="-3.763,-36.051,-10.891") -> (Pose) {
    rotationMode=BMM_Additive
    rotationSpace=BCS_BoneSpace
    alphaBoolBlend {blendOption=Linear}
    boneToModify {boneName=dyn_jacket_fr_mid_MSTR_R}
    rotation=<rot -3.7629 -36.0511 -10.8911>
   }
   n46 = "Transform (Modify) Bone - Bone: dyn_jacket_MSTR_R"!(ComponentPose=@n45, Alpha=1, Rotation="1.878,-29.124,32.747") -> (Pose) {
    rotationMode=BMM_Additive
    rotationSpace=BCS_BoneSpace
    alphaBoolBlend {blendOption=Linear}
    boneToModify {boneName=dyn_jacket_MSTR_R}
    rotation=<rot 1.878 -29.124 32.747>
   }
   Save cached pose 'main pose'(Pose=Component To Local(ComponentPose=@n46)) {cacheName=main pose}
   n47 = "Transform (Modify) Bone - Bone: dyn_jacket_fr_mid_MSTR_L"!(
    ComponentPose=Local To Component(LocalPose=Use cached pose 'main pose'())
    Alpha=1
    Translation:Vector="-11.595,16.711,1.497"
    Rotation="45.000,-179.991,179.990"
   ) -> (Pose) {
    rotationMode=BMM_Additive
    rotationSpace=BCS_BoneSpace
    translationMode=BMM_Additive
    translationSpace=BCS_BoneSpace
    alphaBoolBlend {blendOption=Linear}
    boneToModify {boneName=dyn_jacket_fr_mid_MSTR_L}
    rotation=<rot 44.9998 -179.991 179.99>
    translation=<-11.595 16.711 1.497>
   }
   n48 = "Transform (Modify) Bone - Bone: dyn_jacket_MSTR_L"!(ComponentPose=@n47, Alpha=1, Translation:Vector="-13.471,-8.553,1.665", Rotation="0.145,49.434,0.074") -> (Pose) {
    rotationMode=BMM_Additive
    rotationSpace=BCS_BoneSpace
    translationMode=BMM_Additive
    translationSpace=BCS_BoneSpace
    alphaBoolBlend {blendOption=Linear}
    boneToModify {boneName=dyn_jacket_MSTR_L}
    rotation=<rot 0.145 49.4342 0.074>
    translation=<-13.471 -8.553 1.665>
   }
   Save cached pose 'crouching'(Pose=Component To Local(ComponentPose=@n48)) {cacheName=crouching}
   n49 = "Constraint - dyn_jacket_fr_MSTR (P)"(ComponentPose=Local To Component(LocalPose=Use cached pose 'main pose'()), Alpha=1, ConstraintWeights_0=1) {
    alphaBoolBlend {blendOption=Linear}
    boneToModify {boneName=dyn_jacket_fr_MSTR}
    constraintSetup [{targetBone {boneName=pelvis}}]
    constraintWeights [1]
   }
   n50 = "Constraint - dyn_jacket_fr_mid_MSTR_L (P)"(ComponentPose=@n49, Alpha=1, ConstraintWeights_0=1) {
    alphaBoolBlend {blendOption=Linear}
    boneToModify {boneName=dyn_jacket_fr_mid_MSTR_L}
    constraintSetup [{targetBone {boneName=pelvis}}]
    constraintWeights [1]
   }
   n51 = "Constraint - dyn_jacket_fr_mid_MSTR_R (P)"(ComponentPose=@n50, Alpha=1, ConstraintWeights_0=1) {
    alphaBoolBlend {blendOption=Linear}
    boneToModify {boneName=dyn_jacket_fr_mid_MSTR_R}
    constraintSetup [{targetBone {boneName=pelvis}}]
    constraintWeights [1]
   }
   n52 = "Transform (Modify) Bone - Bone: dyn_jacket_fr_mid_MSTR_L"!(ComponentPose=@n51, Alpha=1, Rotation="37.359,51.363,49.457") -> (Pose) {
    rotationMode=BMM_Additive
    rotationSpace=BCS_BoneSpace
    alphaBoolBlend {blendOption=Linear}
    boneToModify {boneName=dyn_jacket_fr_mid_MSTR_L}
    rotation=<rot 37.359 51.363 49.457>
   }
   n53 = "Transform (Modify) Bone - Bone: dyn_jacket_MSTR_L"!(ComponentPose=@n52, Alpha=1, Rotation="0.000,30.000,0.000") -> (Pose) {
    rotationMode=BMM_Additive
    rotationSpace=BCS_BoneSpace
    alphaBoolBlend {blendOption=Linear}
    boneToModify {boneName=dyn_jacket_MSTR_L}
    rotation=<rot 0 30 0>
   }
   Save cached pose 'grind rail'(Pose=Component To Local(ComponentPose=@n53)) {cacheName=grind rail}
  ]}
 EventGraph {nodes [
   n1 = Event BlueprintUpdateAnimation()
   > Set Delta Time X(Delta Time X=@n1.DeltaTimeX)
   > n2 = Sequence!(execute) -> (then_0, then_1, then_2:Exec, then_3:Exec, then_4:Exec)
   n3 = Gravity Override State Function DEPRECATED(
    execute=@n2.then_1
    CharacterPartAnimInstance=Self-Reference()
    GameTimeInSeconds=GetGameTimeInSeconds()
    GravityOverrideParams=Get Gravity Override Params()
    RBAN Params=Get RBAN Params()
   )
   > Set Gravity Override(Gravity Override=@n3)
   > Set RBAN Settings Out(RBAN Settings Out:RBANSetting Struct Structure=@n3.RBAN SettingsOut) -> (Output_Get:RBANSetting Struct Structure)
   n4 = GetOwnerSkeletalMeshComponent()
   n5 = BreakRotator(InRot=GetDeltaTransformFromRefPose!(self=@n4, BoneName=hand_l, BaseName=lowerarm_l) -> (ReturnValue_Location:Vector, ReturnValue_Rotation:Rotator, ReturnValue_Scale:Vector).ReturnValue_Rotation)
   Set hand l pitch(execute=@n2.then_4, hand l pitch:double=MapRangeClamped(Value=@n5.Pitch, InRangeA=0, InRangeB=65)) -> (Output_Get:double)
   > Set hand l yaw(hand l yaw:double=MapRangeClamped(Value=@n5.Yaw, InRangeA=0, InRangeB=60)) -> (Output_Get:double)
   n6 = BreakRotator(InRot=GetDeltaTransformFromRefPose(self=@n4, BoneName=hand_r, BaseName=lowerarm_r).ReturnValue_Rotation)
   > Set hand r pitch(hand r pitch:double=MapRangeClamped(Value=@n6.Pitch, InRangeA=0, InRangeB=65)) -> (Output_Get:double)
   > Set hand r yaw(hand r yaw:double=MapRangeClamped(Value=@n6.Yaw, InRangeA=0, InRangeB=60)) -> (Output_Get:double)
   n7 = NOT Boolean(A=Get bIsSkyDiving())
   n8 = OR Boolean(A=Get bIsJumping(), B=Get bIsFalling())
   Set is jumping(execute=@n2.then_0, is jumping:double=SelectFloat(A=1, B=0, bPickA=AND Boolean(A=@n8, B=@n7))) -> (Output_Get:double)
   n9 = SelectFloat(A=5, B=0.03, bPickA="float < float"(A=Get Is Sky Diving() -> (Is Sky Diving:double), B=0.95))
   n10 = Get Delta Time X()
   n11 = Get Is Sky Diving()
   n12 = Get bIsSkydiveFloating()
   n13 = Get bIsSkydiveDiveMode()
   > Set Is Sky Diving(Is Sky Diving:double=SelectFloat(A=1, B=FInterpTo(Current=@n11, Target=0, DeltaTime=@n10, InterpSpeed=@n9), bPickA=OR Boolean(A=@n12, B=@n13))) -> (Output_Get:double)
   n14 = BreakRotator(InRot=GetDeltaTransformFromRefPose(self=GetOwnerSkeletalMeshComponent(), BoneName=spine_05, BaseName=head).ReturnValue_Rotation)
   Set Head Roll L Compensation(execute=@n2.then_3, Head Roll L Compensation:double=MapRangeClamped(Value=@n14, InRangeA=0, InRangeB=-70, OutRangeA=1, OutRangeB=0)) -> (Output_Get:double)
   > Set Head Roll R Compensation(Head Roll R Compensation:double=MapRangeClamped(Value=@n14, InRangeA=0, InRangeB=70, OutRangeA=1, OutRangeB=0)) -> (Output_Get:double)
   n15 = Get Head Roll L Compensation() -> (Head Roll L Compensation:double)
   n16 = Get Head Roll R Compensation() -> (Head Roll R Compensation:double)
   > Set Head Roll Compensation(Head Roll Compensation:double=MapRangeClamped(Value=float + float(A=@n15, B=@n16), InRangeA=2, InRangeB=0, OutRangeA=1, OutRangeB=0.5)) -> (Output_Get:double)
   > Set Head Yaw Compensation(Head Yaw Compensation:double=MapRangeClamped(Value=@n14.Yaw, InRangeA=0, InRangeB=30, OutRangeA=1, OutRangeB=0)) -> (Output_Get:double)
   n17 = Get Head Roll L Compensation()
   > Set Head Pitch L(Head Pitch L:double=MapRangeClamped(Value=@n14.Pitch, InRangeA=0, InRangeB=15, OutRangeB=MapRangeClamped(Value=@n17, InRangeA=1, InRangeB=0, OutRangeA=0.3))) -> (Output_Get:double)
   n18 = Get Head Roll R Compensation()
   > Set Head Pitch R(Head Pitch R:double=MapRangeClamped(Value=@n14.Pitch, InRangeA=0, InRangeB=-15, OutRangeB=MapRangeClamped(Value=@n18, InRangeA=1, InRangeB=0, OutRangeA=0.3))) -> (Output_Get:double)
   n19 = Get Head Yaw Compensation() -> (Head Yaw Compensation:double)
   > Set Head Roll L(Head Roll L:double=MapRangeClamped(Value=@n14, InRangeA=0, InRangeB=-70, OutRangeB=@n19)) -> (Output_Get:double)
   > Set Head Roll R(Head Roll R:double=MapRangeClamped(Value=@n14, InRangeA=0, InRangeB=70, OutRangeB=@n19)) -> (Output_Get:double)
   > Set Head Yaw(Head Yaw:double=MapRangeClamped(Value=@n14.Yaw, InRangeA=0, InRangeB=50, OutRangeB=Get Head Roll Compensation() -> (Head Roll Compensation:double))) -> (Output_Get:double)
   > Set Head Yaw L(Head Yaw L:double=MapRangeClamped(Value=@n14.Yaw, InRangeA=0, InRangeB=50, OutRangeB=Get Head Roll L Compensation())) -> (Output_Get:double)
   > Set Head Yaw R(Head Yaw R:double=MapRangeClamped(Value=@n14.Yaw, InRangeA=0, InRangeB=50, OutRangeB=Get Head Roll R Compensation())) -> (Output_Get:double)
   > Set Extra Head Yaw(Extra Head Yaw:double=MapRangeClamped(Value=@n14.Yaw, InRangeA=50, InRangeB=90)) -> (Output_Get:double)
   n20 = GetOwnerSkeletalMeshComponent()
   n21 = BreakRotator(InRot=GetDeltaTransformFromRefPose(self=@n20, BoneName=thigh_l, BaseName=pelvis).ReturnValue_Rotation)
   Set thigh l yaw extra(execute=@n2.then_2, thigh l yaw extra:double=MapRangeClamped(Value=@n21.Yaw, InRangeA=-100, InRangeB=-150)) -> (Output_Get:double)
   n22 = SelectFloat(A=0, B=-50, bPickA=Get bIsOnGrindRail())
   > Set thigh l yaw(thigh l yaw:double=MapRangeClamped(Value=@n21.Yaw, InRangeA=@n22, InRangeB=-100)) -> (Output_Get:double)
   n23 = BreakRotator(InRot=GetDeltaTransformFromRefPose(self=@n20, BoneName=thigh_r, BaseName=pelvis).ReturnValue_Rotation)
   > Set thigh r yaw(thigh r yaw:double=MapRangeClamped(Value=@n23.Yaw, InRangeA=@n22, InRangeB=-100)) -> (Output_Get:double)
   > Set thigh r yaw extra(thigh r yaw extra:double=MapRangeClamped(Value=@n23.Yaw, InRangeA=-100, InRangeB=-150)) -> (Output_Get:double)
   n24 = Get thigh l yaw() -> (thigh l yaw:double)
   n25 = Get thigh r yaw() -> (thigh r yaw:double)
   > Set any thigh yaw(any thigh yaw:double=MapRangeClamped(Value=float + float(A=@n24, B=@n25), InRangeA=0, InRangeB=1)) -> (Output_Get:double)
  ]}
]
variables [
 any thigh yaw {type=number}
 Delta Time X {type=number}
 Extra Head Yaw {type=number}
 Gravity Override {type=Vector}
 Gravity Override Params {
  type=GravityOverrideParamsStruct
  default {
   skydive_FloatingDown {gravityOverride=<0 -350 300>, windAmplitude=<100 -250 200>, windFrequency=<1.5 5 6>, windOffset=<0 -300 200>}
   skydive_GoingUp {gravityOverride=<0 -100 500>, windAmplitude=<100 -250 200>, windFrequency=<1.5 5 6>}
   skydive_Parachuting {gravityOverride=<0 -400 -500>, windAmplitude=<100 -250 200>, windFrequency=<1.5 5 6>}
  }
 }
 hand l pitch {type=number}
 hand l yaw {type=number}
 hand r pitch {type=number}
 hand r yaw {type=number}
 Head Pitch L {type=number}
 Head Pitch R {type=number}
 Head Roll Compensation {type=number}
 Head Roll L {type=number}
 Head Roll L Compensation {type=number}
 Head Roll R {type=number}
 Head Roll R Compensation {type=number}
 Head Yaw {type=number}
 Head Yaw Compensation {type=number}
 Head Yaw L {type=number}
 Head Yaw R {type=number}
 is jumping {type=number}
 Is Sky Diving {type=number}
 RBAN Params {
  type=RBANParamsStruct
  default {
   dBNO {
    componentLinearAccScale=<0.3 0.3 0.3>
    componentLinearVelScale=<0.3 0.3 0.3>
    simSpaceSettings {
     maxAngularAcceleration=1000
     maxAngularVelocity=3
     maxLinearAcceleration=300
     maxLinearVelocity=700
     velocityScaleZ=0
     externalLinearDragV=<0 0.5 0>
    }
   }
   dBNO_Carried {
    componentLinearAccScale=<0.3 0.3 0.3>
    componentLinearVelScale=<0.3 0.3 0.3>
    simSpaceSettings {
     maxAngularAcceleration=1000
     maxAngularVelocity=3
     maxLinearAcceleration=300
     maxLinearVelocity=700
     velocityScaleZ=0
     externalLinearDragV=<0 0.5 0>
    }
   }
   emoteOrMelee {
    componentLinearAccScale=<0.3 0.3 0.3>
    componentLinearVelScale=<0.3 0.3 0.3>
    simSpaceSettings {
     maxAngularAcceleration=1000
     maxAngularVelocity=3
     maxLinearAcceleration=300
     maxLinearVelocity=700
     velocityScaleZ=0
     externalLinearDragV=<0 0.5 0>
    }
   }
   front_End {
    componentLinearAccScale=<0.3 0.3 0.3>
    componentLinearVelScale=<0.3 0.3 0.3>
    simSpaceSettings {
     maxAngularAcceleration=1000
     maxAngularVelocity=3
     maxLinearAcceleration=300
     maxLinearVelocity=700
     velocityScaleZ=0
     externalLinearDragV=<0 0.5 0>
    }
   }
   onGround_Crouch_Moving {
    componentLinearAccScale=<0.3 0.3 0.3>
    componentLinearVelScale=<0.3 0.3 0.3>
    simSpaceSettings {
     maxAngularAcceleration=1000
     maxAngularVelocity=3
     maxLinearAcceleration=300
     maxLinearVelocity=700
     velocityScaleZ=0
     externalLinearDragV=<0 0.5 0>
    }
   }
   onGround_Falling {
    componentLinearAccScale=<0.3 0.3 0.3>
    componentLinearVelScale=<0.3 0.3 0.3>
    simSpaceSettings {
     maxAngularAcceleration=1000
     maxAngularVelocity=3
     maxLinearAcceleration=300
     maxLinearVelocity=700
     velocityScaleZ=0
     externalLinearDragV=<0 0.5 0>
    }
   }
   onGround_Moving {
    componentLinearAccScale=<0.3 0.3 0.3>
    componentLinearVelScale=<0.3 0.3 0.3>
    simSpaceSettings {
     maxAngularAcceleration=1000
     maxAngularVelocity=3
     maxLinearAcceleration=300
     maxLinearVelocity=700
     velocityScaleZ=0
     externalLinearDragV=<0 0.5 0>
    }
   }
   onGround_Standing {
    componentLinearAccScale=<0.3 0.3 0.3>
    componentLinearVelScale=<0.3 0.3 0.3>
    simSpaceSettings {
     maxAngularAcceleration=1000
     maxAngularVelocity=3
     maxLinearAcceleration=300
     maxLinearVelocity=700
     velocityScaleZ=0
     externalLinearDragV=<0 0.5 0>
    }
   }
   rail_Grinding {
    componentLinearAccScale=<0.3 0.3 0.3>
    componentLinearVelScale=<0.3 0.3 0.3>
    simSpaceSettings {
     maxAngularAcceleration=1000
     maxAngularVelocity=3
     maxLinearAcceleration=300
     maxLinearVelocity=700
     velocityScaleZ=0
     worldAlpha=1
     externalLinearDragV=<0 0.5 0>
    }
   }
   riding_Moving {
    componentLinearAccScale=<0.3 0.3 0.3>
    componentLinearVelScale=<0.3 0.3 0.3>
    simSpaceSettings {
     maxAngularAcceleration=1000
     maxAngularVelocity=3
     maxLinearAcceleration=300
     maxLinearVelocity=700
     velocityScaleZ=0
     externalLinearDragV=<0 0.5 0>
    }
   }
   riding_Moving_Targeting {
    componentLinearAccScale=<0.3 0.3 0.3>
    componentLinearVelScale=<0.3 0.3 0.3>
    simSpaceSettings {
     maxAngularAcceleration=1000
     maxAngularVelocity=3
     maxLinearAcceleration=300
     maxLinearVelocity=700
     velocityScaleZ=0
     externalLinearDragV=<0 0.5 0>
    }
   }
   riding_Standing {
    componentLinearAccScale=<0.3 0.3 0.3>
    componentLinearVelScale=<0.3 0.3 0.3>
    simSpaceSettings {
     maxAngularAcceleration=1000
     maxAngularVelocity=3
     maxLinearAcceleration=300
     maxLinearVelocity=700
     velocityScaleZ=0
     externalLinearDragV=<0 0.5 0>
    }
   }
   skydive_DivingDown {
    componentLinearAccScale=<0.1 0.1 0.1>
    componentLinearVelScale=<0.1 0.1 0.1>
    simSpaceSettings {
     maxAngularAcceleration=1000
     maxAngularVelocity=3
     maxLinearAcceleration=300
     maxLinearVelocity=700
     velocityScaleZ=0
     externalLinearDragV=<0 0.5 0>
    }
   }
   skydive_FloatingDown {
    componentLinearAccScale=<0.1 0.1 0.1>
    componentLinearVelScale=<0.1 0.1 0.1>
    simSpaceSettings {
     maxAngularAcceleration=1000
     maxAngularVelocity=3
     maxLinearAcceleration=300
     maxLinearVelocity=700
     velocityScaleZ=0
     externalLinearDragV=<0 0.5 0>
    }
   }
   skydive_GoingUp {
    componentLinearAccScale=<0.1 0.1 0.1>
    componentLinearVelScale=<0.1 0.1 0.1>
    simSpaceSettings {
     maxAngularAcceleration=1000
     maxAngularVelocity=3
     maxLinearAcceleration=300
     maxLinearVelocity=700
     velocityScaleZ=0
     externalLinearDragV=<0 0.5 0>
    }
   }
   skydive_Parachuting {
    componentLinearAccScale=<0.1 0.1 0.1>
    componentLinearVelScale=<0.1 0.1 0.1>
    simSpaceSettings {
     maxAngularAcceleration=1000
     maxAngularVelocity=3
     maxLinearAcceleration=300
     maxLinearVelocity=700
     velocityScaleZ=0
     externalLinearDragV=<0 0.5 0>
    }
   }
   skydive_Surfing {
    componentLinearAccScale=<0.3 0.3 0.3>
    componentLinearVelScale=<0.3 0.3 0.3>
    simSpaceSettings {
     maxAngularAcceleration=1000
     maxAngularVelocity=3
     maxLinearAcceleration=300
     maxLinearVelocity=700
     velocityScaleZ=0
     externalLinearDragV=<0 0.5 0>
    }
   }
   swimming {
    componentLinearAccScale=<0.3 0.3 0.3>
    componentLinearVelScale=<0.3 0.3 0.3>
    simSpaceSettings {
     maxAngularAcceleration=1000
     maxAngularVelocity=3
     maxLinearAcceleration=300
     maxLinearVelocity=700
     velocityScaleZ=0
     externalLinearDragV=<0 0.5 0>
    }
   }
   tactical_Sprint {
    componentLinearAccScale=<0.3 0.3 0.3>
    componentLinearVelScale=<0.3 0.3 0.3>
    simSpaceSettings {
     maxAngularAcceleration=1000
     maxAngularVelocity=3
     maxLinearAcceleration=300
     maxLinearVelocity=700
     velocityScaleZ=0
     externalLinearDragV=<0 0.5 0>
    }
   }
   ziplining {
    componentLinearAccScale=<0.3 0.3 0.3>
    componentLinearVelScale=<0.3 0.3 0.3>
    simSpaceSettings {
     maxAngularAcceleration=1000
     maxAngularVelocity=3
     maxLinearAcceleration=300
     maxLinearVelocity=700
     velocityScaleZ=0
     externalLinearDragV=<0 0.5 0>
    }
   }
  }
 }
 RBAN Settings Out {
  type=RBANSettingStruct
  default {
   componentLinearAccScale=<0 0 0>
   componentLinearVelScale=<0 0 0>
   simSpaceSettings {
    dampingAlpha=1
    maxAngularAcceleration=1000
    maxAngularVelocity=3
    maxLinearAcceleration=300
    maxLinearVelocity=700
    velocityScaleZ=0
    worldAlpha=1
    externalAngularVelocity=<0 0 0>
    externalLinearDragV=<0 0.5 0>
    externalLinearVelocity=<0 0 0>
   }
  }
 }
 thigh l yaw {type=number}
 thigh l yaw extra {type=number}
 thigh r yaw {type=number}
 thigh r yaw extra {type=number}
]
